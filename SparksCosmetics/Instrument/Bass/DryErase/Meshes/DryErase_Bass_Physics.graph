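# Physics asset for the DryErase bass instrument cosmetic, simulating a dangling loop and tag attached to its handle.
# Three capsule bodies cover the handle, dyn_loop and dyn_tag bones.
# Constraints form a chain: the loop joins the handle with locked twist and limited swing, and the tag joins the loop with much wider swing and slight twist.
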PhysicsAsset /SparksCosmetics/Instrument/Bass/DryErase/Meshes/DryErase_Bass_Physics
bodies {
 dyn_loop [{shapeType=Capsule, length=0.7251, radius=3.2849, shapeName=dyn_loop_capsule, center=<2.8337 0 0>, rotation=<rot 90 0 0>}]
 dyn_tag [{shapeType=Capsule, length=1.9664, radius=1.5486, shapeName=dyn_tag_capsule, center=<1.6152 -0.0253 0.0735>, rotation=<rot 0 90 -85.1058>}]
 handle [{shapeType=Capsule, length=5, radius=5, rotation=<rot 0 -90 0>}]
}
constraints [
 {bone1Name=dyn_loop, bone2Name=handle, swing1LimitDegrees=2, swing2LimitDegrees=30, twistLimitDegrees=0, twistMotion=Locked}
 {bone1Name=dyn_tag, bone2Name=dyn_loop, swing1LimitDegrees=80, swing2LimitDegrees=70, twistLimitDegrees=10}
]
